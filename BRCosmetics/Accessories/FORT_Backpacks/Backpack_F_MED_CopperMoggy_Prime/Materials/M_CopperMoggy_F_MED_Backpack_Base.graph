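# Base material for the CopperMoggy Prime backpack cosmetic, built on the 2023 character parent material functions with material attributes.
# The PreFX function feeds a generic visor function for TV static and glitch effects, which also overrides PreFX's basic emissive UVs, then the PostFX function finishes the result.
# An EnableFlameOverrides static switch optionally inserts the CopperMoggy flame overrides before PostFX and routes their temporal mask to Temporal Responsiveness output.
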Material /BRCosmetics/Accessories/FORT_Backpacks/Backpack_F_MED_CopperMoggy_Prime/Materials/M_CopperMoggy_F_MED_Backpack_Base
graph {
 properties {
  bUseMaterialAttributes=true
  bUsedWithClothing=true
  bUsedWithMeshDeformer=true
  bUsedWithMorphTargets=true
  bUsedWithNanite=true
  bUsedWithNiagaraMeshParticles=true
  bUsedWithSkeletalMesh=true
  pixelDepthOffsetMode=PDOM_Legacy
  shadingModel=MSM_FromMaterialExpression
  usedShadingModels=Default Lit | Subsurface | Clear Coat | Subsurface Profile | Cloth
 }
 nodes [
  n1 = MaterialFunctionCall(Override Basic Emissive UVs=@n2.UVs) {materialFunction=&MF_CharacterParent_2023_PreFX}
  n2 = MaterialFunctionCall!(Material=@n1.Result, Tech Art Mask=@n1.Result) {desc=Used for TV static / glitch effects, materialFunction=/Game/Global/Materials/Character_Parent_2023/Visor_Parent_2025/MF_GenericVisor}
  n3 = MaterialFunctionCall!(Material=@n2.Result, TechArtMask=@n1.Tech Art Mask) {materialFunction=/BRCosmetics/Characters/Player/Female/Medium/Bodies/F_MED_CopperMoggy_Swan/Skins/Flame/Materials/MF_CopperMoggy_FlameOverrides}
  MaterialFunctionCall(
   Material=StaticSwitchParameter(True=@n3.Result, False=@n2.Result) {group=Flame, parameterName=EnableFlameOverrides}
   M=@n1.M
   Pre Skinned Local Position=@n1.M
   Distance Blend=@n1.M
   Fresnel=@n1.M
  ) {materialFunction=&MF_CharacterParent_2023_PostFX}
  TemporalResponsivenessOutput!(None=StaticSwitchParameter(True=@n3.Temporal Mask, False=Constant()) {group=Flame, parameterName=EnableFlameOverrides})
 ]
}
