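# Physics asset for the CopperMoggy Prime backpack cosmetic, defining a short simulated belt chain.
# A sphere body on base_belt anchors three capsule bodies on dyn_belt_01 to 03, each constrained to the previous bone.
# All joints lock twist and allow limited swing; the first swings on one axis only, while the lower two also get a narrower second swing axis.
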PhysicsAsset /BRCosmetics/Accessories/FORT_Backpacks/Backpack_F_MED_CopperMoggy_Prime/Meshes/F_MED_CopperMoggy_Prime_Pack_Physics4
bodies {
 base_belt [{shapeType=Sphere, radius=3, shapeName=base_belt_sphere}]
 dyn_belt_01 [{shapeType=Capsule, length=6, radius=4, shapeName=dyn_belt_01_capsule, center=<5 0 0>, rotation=<rot -90 0 0>}]
 dyn_belt_02 [{shapeType=Capsule, length=6, radius=4, shapeName=dyn_belt_02_capsule, center=<5 0 0>, rotation=<rot -90 0 0>}]
 dyn_belt_03 [{shapeType=Capsule, length=6, radius=4, shapeName=dyn_belt_03_capsule, center=<5 0 0>, rotation=<rot -90 0 0>}]
}
constraints [
 {bone1Name=dyn_belt_01, bone2Name=base_belt, swing1LimitDegrees=30, swing2LimitDegrees=0, twistLimitDegrees=0}
 {bone1Name=dyn_belt_02, bone2Name=dyn_belt_01, swing1LimitDegrees=30, swing2LimitDegrees=10, twistLimitDegrees=0}
 {bone1Name=dyn_belt_03, bone2Name=dyn_belt_02, swing1LimitDegrees=30, swing2LimitDegrees=10, twistLimitDegrees=0}
]
