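# Animation Blueprint for the CopperMoggy Prime backpack cosmetic, adding secondary motion to its belt, glasses and compass.
# Glasses and compass Anim Dynamics feed a cached pose, which branches into belt-only Anim Dynamics for the front end or the in-game chain.
# In-game, RigidBody simulation precedes additive pack and belt corrections driven by spine rotation, with belt Anim Dynamics when rigid bodies are disabled.
# The EventGraph also updates gravity override and sim-space settings.
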AnimBlueprint /BRCosmetics/Accessories/FORT_Backpacks/Backpack_F_MED_CopperMoggy_Prime/Meshes/F_MED_CopperMoggy_Prime_Pack_AnimBP
target_skeleton=&/F_MED_CopperMoggy_Prime_Pack_Skeleton
graphs [
 AnimGraph {nodes [
   n1 = RigidBody!(
    ComponentPose=Local To Component(LocalPose=Use cached pose 'SavedPose25880'())
    Alpha
    OverrideWorldGravity:Vector="0,0,-980"
    ExternalForce
    SimSpaceSettings:Sim Space Settings Structure=Get Sim Space Settings() -> (Sim Space Settings:Sim Space Settings Structure)
   ) -> (Pose) {
    bClampLinearTranslationLimitToRefPose=true
    bOverrideWorldGravity=true
    overridePhysicsAsset=&/F_MED_CopperMoggy_Prime_Pack_Physics4
    simulationSpace=BaseBoneSpace
    baseBoneRef {boneName=base_belt}
    overrideWorldGravity {z=-980}
    simSpaceSettings {
     dampingAlpha=0
     maxAngularAcceleration=10
     maxAngularVelocity=2
     maxLinearAcceleration=1200
     maxLinearVelocity=300
     worldAlpha=1
     externalLinearDragV {z=1}
    }
   }
   n2 = "Transform (Modify) Bone - Bone: dyn_belt_02"!(
    ComponentPose=@n1
    bAlphaBoolEnabled:bool=Get RBAN_of() -> (RBAN_of:bool)
    Translation="0,0,0"
    Rotation=Get belt_pickaxe_Y() -> (belt_pickaxe_Y:Rotator)
    Scale="1,1,1"
   ) -> (Pose) {alphaInputType=Bool, rotationMode=BMM_Additive, rotationSpace=BCS_BoneSpace, boneToModify {boneName=dyn_belt_02}}
   n3 = Get RBAN_on() -> (RBAN_on:Rotator)
   n4 = "Transform (Modify) Bone - Bone: dyn_belt_04"(
    ComponentPose="Transform (Modify) Bone - Bone: pack"(
     ComponentPose="Transform (Modify) Bone - Bone: dyn_belt_02"!(ComponentPose=@n2, Alpha:float=1, Translation="0,0,0", Rotation=@n3, Scale="1,1,1") -> (Pose) {rotationMode=BMM_Additive, rotationSpace=BCS_BoneSpace, boneToModify {boneName=dyn_belt_02}}
     Translation="0,0.3,0"
     Rotation="0,0,0"
    ) {translationMode=BMM_Additive, translationSpace=BCS_BoneSpace, boneToModify {boneName=pack}, translation {y=0.3}}
    Alpha=0
    Translation="0,0,0"
    Rotation=Get belt_rotate() -> (belt_rotate:Rotator)
    Scale="1,1,1"
   ) {
    rotationMode=BMM_Additive
    translationMode=BMM_Additive
    translationSpace=BCS_BoneSpace
    alphaBoolBlend {blendOption=Linear}
    boneToModify {boneName=dyn_belt_04}
   }
   n5 = "Transform (Modify) Bone - Bone: dyn_belt_03"(ComponentPose=@n4, Alpha=1, Translation="0,0,0", Rotation=Get belt_rotate(), Scale="1,1,1") {rotationMode=BMM_Additive, translationMode=BMM_Additive, alphaBoolBlend {blendOption=Linear}, boneToModify {boneName=dyn_belt_03}}
   n6 = "Transform (Modify) Bone - Bone: dyn_belt_02"(ComponentPose=@n5, Alpha=1, Translation="0,0,0", Rotation=Get belt_rotate(), Scale="1,1,1") {rotationMode=BMM_Additive, translationMode=BMM_Additive, alphaBoolBlend {blendOption=Linear}, boneToModify {boneName=dyn_belt_02}}
   n7 = "Transform (Modify) Bone - Bone: pack"(ComponentPose=@n6, Translation="0,0,0", Rotation=Get pickaxe_rotate() -> (pickaxe_rotate:Rotator)) {rotationMode=BMM_Additive, translationMode=BMM_Additive, alphaBoolBlend {blendOption=Linear}, boneToModify {boneName=pack}}
   n8 = "Transform (Modify) Bone - Bone: pack"(
    ComponentPose=@n7
    Translation=Get backpack_translation()
    Rotation=Get backpack_rotate()
    Scale=Get backpack_scale() -> (backpack_scale:Vector)
   ) {
    rotationMode=BMM_Additive
    scaleMode=BMM_Additive
    translationMode=BMM_Additive
    alphaBoolBlend {blendOption=Linear}
    boneToModify {boneName=pack}
   }
   n9 = Get RBAN_of()
   n10 = Anim Dynamics - Bone: dyn_belt_01!(ComponentPose=@n8, bAlphaBoolEnabled:bool=@n9, ExternalForce="0,0,0") -> (Pose) {
    alphaInputType=Bool
    angularDampingOverride=0.999
    angularSpringConstant=250
    bAngularSpring=true
    bOverrideAngularDamping=true
    bUseGravityOverride=true
    simulationSpace=RootRelative
    boundBone {boneName=dyn_belt_01}
    componentAppliedLinearAccClamp=<3000 3000 3000>
    componentLinearAccScale=<0.4 0.4 0.4>
    componentLinearVelScale=<0.4 0.4 0.4>
    gravityOverride {z=-980}
    physicsBodyDefinitions [{
      boundBone {boneName=dyn_belt_01}
      boxExtents=<10 10 10>
      constraintSetup {angularLimitsMax=<0 0 10>, angularLimitsMin=<0 0 -45>, angularTarget=<1 0 0>}
      localJointOffset=<0.8 0 0>
     }]
   }
   n11 = Anim Dynamics - Bone: dyn_belt_02!(ComponentPose=@n10, bAlphaBoolEnabled:bool=@n9, ExternalForce="0,0,0") -> (Pose) {
    alphaInputType=Bool
    angularDampingOverride=0.999
    angularSpringConstant=250
    bAngularSpring=true
    bOverrideAngularDamping=true
    bUseGravityOverride=true
    simulationSpace=RootRelative
    boundBone {boneName=dyn_belt_02}
    componentAppliedLinearAccClamp=<3000 3000 3000>
    componentLinearAccScale=<0.4 0.4 0.4>
    componentLinearVelScale=<0.4 0.4 0.4>
    gravityOverride {z=-980}
    physicsBodyDefinitions [{
      boundBone {boneName=dyn_belt_02}
      boxExtents=<10 10 10>
      constraintSetup {angularLimitsMax=<3 5 10>, angularLimitsMin=<-3 -5 -45>, angularTarget=<1 0 0>}
      localJointOffset=<0.8 0 0>
     }]
   }
   n12 = Anim Dynamics - Bone: dyn_belt_03!(ComponentPose=@n11, bAlphaBoolEnabled:bool=@n9, ExternalForce="0,0,0") -> (Pose) {
    alphaInputType=Bool
    angularDampingOverride=0.999
    angularSpringConstant=250
    bAngularSpring=true
    bOverrideAngularDamping=true
    bUseGravityOverride=true
    simulationSpace=RootRelative
    boundBone {boneName=dyn_belt_03}
    componentAppliedLinearAccClamp=<3000 3000 3000>
    componentLinearAccScale=<0.4 0.4 0.4>
    componentLinearVelScale=<0.4 0.4 0.4>
    gravityOverride {z=-980}
    physicsBodyDefinitions [{
      boundBone {boneName=dyn_belt_03}
      boxExtents=<10 10 10>
      constraintSetup {angularLimitsMax=<3 5 10>, angularLimitsMin=<-3 -5 -45>, angularTarget=<1 0 0>}
      localJointOffset=<0.8 0 0>
     }]
   }
   n13 = Anim Dynamics - Bone: dyn_belt_01!(ComponentPose=Local To Component(LocalPose=Use cached pose 'SavedPose25880'()), Alpha:float=1, ExternalForce="0,0,0") -> (Pose) {
    angularDampingOverride=0.999
    angularSpringConstant=250
    bAngularSpring=true
    bOverrideAngularDamping=true
    bUseGravityOverride=true
    simulationSpace=RootRelative
    boundBone {boneName=dyn_belt_01}
    componentAppliedLinearAccClamp=<3000 3000 3000>
    componentLinearAccScale=<0.4 0.4 0.4>
    componentLinearVelScale=<0.4 0.4 0.4>
    gravityOverride {z=-980}
    physicsBodyDefinitions [{
      boundBone {boneName=dyn_belt_01}
      boxExtents=<10 10 10>
      constraintSetup {angularLimitsMax=<0 0 10>, angularLimitsMin=<0 0 -45>, angularTarget=<1 0 0>}
      localJointOffset=<0.5 0 0>
     }]
   }
   n14 = Anim Dynamics - Bone: dyn_belt_02!(ComponentPose=@n13, Alpha:float=1, ExternalForce="0,0,0") -> (Pose) {
    angularDampingOverride=0.999
    angularSpringConstant=250
    bAngularSpring=true
    bOverrideAngularDamping=true
    bUseGravityOverride=true
    simulationSpace=RootRelative
    boundBone {boneName=dyn_belt_02}
    componentAppliedLinearAccClamp=<3000 3000 3000>
    componentLinearAccScale=<0.4 0.4 0.4>
    componentLinearVelScale=<0.4 0.4 0.4>
    gravityOverride {z=-980}
    physicsBodyDefinitions [{
      boundBone {boneName=dyn_belt_02}
      boxExtents=<10 10 10>
      constraintSetup {angularLimitsMax=<3 5 10>, angularLimitsMin=<-3 -5 -45>, angularTarget=<1 0 0>}
      localJointOffset=<1 0 0>
     }]
   }
   n15 = Anim Dynamics - Bone: dyn_belt_03!(ComponentPose=@n14, Alpha:float=1, ExternalForce="0,0,0") -> (Pose) {
    angularDampingOverride=0.999
    angularSpringConstant=250
    bAngularSpring=true
    bOverrideAngularDamping=true
    bUseGravityOverride=true
    simulationSpace=RootRelative
    boundBone {boneName=dyn_belt_03}
    componentAppliedLinearAccClamp=<3000 3000 3000>
    componentLinearAccScale=<0.4 0.4 0.4>
    componentLinearVelScale=<0.4 0.4 0.4>
    gravityOverride {z=-980}
    physicsBodyDefinitions [{
      boundBone {boneName=dyn_belt_03}
      boxExtents=<10 10 10>
      constraintSetup {angularLimitsMax=<3 5 10>, angularLimitsMin=<-3 -5 -45>, angularTarget=<1 0 0>}
      localJointOffset=<1 0 0>
     }]
   }
   Output Pose(Result=Blend Poses by bool(BlendPose_0=Component To Local(ComponentPose=@n15), BlendPose_1=Component To Local(ComponentPose=@n12), bActiveValue=Get bIsInFrontEnd())) {layerGroup=None}
   n16 = Anim Dynamics - Bone: dyn_glasses_03(ComponentPose=Local To Component(LocalPose=Input Pose()), Alpha=1, ExternalForce="0,0,0") {
    angularDampingOverride=0.999
    angularSpringConstant=250
    bAngularSpring=true
    bOverrideAngularDamping=true
    bUseGravityOverride=true
    simulationSpace=RootRelative
    boundBone {boneName=dyn_glasses_03}
    componentAppliedLinearAccClamp=<1000 1000 1000>
    componentLinearAccScale=<0.3 0.3 0.3>
    componentLinearVelScale=<0.5 0.5 0.5>
    gravityOverride {z=-980}
    physicsBodyDefinitions [{
      boundBone {boneName=dyn_glasses_03}
      boxExtents=<13 13 13>
      constraintSetup {angularLimitsMax=<0 5 0>, angularLimitsMin=<-5 -30 -35>, angularTarget=<1 0 0>}
      localJointOffset=<1 0 0>
     }]
   }
   n17 = Anim Dynamics - Bone: dyn_glasses_02(ComponentPose=@n16, Alpha=1, ExternalForce="0,0,0") {
    angularDampingOverride=0.999
    angularSpringConstant=250
    bAngularSpring=true
    bOverrideAngularDamping=true
    bUseGravityOverride=true
    simulationSpace=RootRelative
    boundBone {boneName=dyn_glasses_02}
    componentAppliedLinearAccClamp=<1000 1000 1000>
    componentLinearAccScale=<0.3 0.3 0.3>
    componentLinearVelScale=<0.5 0.5 0.5>
    gravityOverride {z=-980}
    physicsBodyDefinitions [{
      boundBone {boneName=dyn_glasses_02}
      boxExtents=<13 13 13>
      constraintSetup {angularLimitsMax=<0 15 35>, angularLimitsMin=<0 0 0>, angularTarget=<1 0 0>}
      localJointOffset=<1 0 0>
     }]
   }
   n18 = Anim Dynamics - Bone: dyn_glasses_01(ComponentPose=@n17, Alpha=1, ExternalForce="0,0,0") {
    angularDampingOverride=0.999
    angularSpringConstant=250
    bAngularSpring=true
    bOverrideAngularDamping=true
    bUseGravityOverride=true
    simulationSpace=RootRelative
    boundBone {boneName=dyn_glasses_01}
    componentAppliedLinearAccClamp=<1000 1000 1000>
    componentLinearAccScale=<0.3 0.3 0.3>
    componentLinearVelScale=<0.5 0.5 0.5>
    gravityOverride {z=-980}
    physicsBodyDefinitions [{
      boundBone {boneName=dyn_glasses_01}
      boxExtents=<13 13 13>
      constraintSetup {angularLimitsMax=<0 0 0>, angularLimitsMin=<0 -5 -35>, angularTarget=<1 0 0>}
      localJointOffset=<1 0 0>
     }]
   }
   n19 = Anim Dynamics - Bone: dyn_compas_01(ComponentPose=@n18, Alpha=1, ExternalForce="0,0,0") {
    angularDampingOverride=0.9999
    angularSpringConstant=250
    bAngularSpring=true
    bOverrideAngularDamping=true
    bUseGravityOverride=true
    simulationSpace=RootRelative
    boundBone {boneName=dyn_compas_01}
    componentAppliedLinearAccClamp=<1000 1000 1000>
    componentLinearAccScale=<0.3 0.6 0.5>
    componentLinearVelScale=<0.5 0.5 0.5>
    gravityOverride {z=-980}
    physicsBodyDefinitions [{
      boundBone {boneName=dyn_compas_01}
      boxExtents=<10 10 10>
      constraintSetup {angularLimitsMax=<0 45 0>, angularLimitsMin=<0 0 -45>, angularTarget=<1 0 0>}
      localJointOffset=<1 0 0>
     }]
   }
   Save cached pose 'SavedPose25880'(Pose=Component To Local(ComponentPose=@n19)) {cacheName=SavedPose25880}
  ]}
 EventGraph {nodes [
   Event BlueprintUpdateAnimation()
   > n1 = Sequence!(execute) -> (then_0, then_1, then_2:Exec, then_3:Exec, then_4:Exec)
   n2 = GetOwnerSkeletalMeshComponent()
   n3 = BreakRotator(InRot=BreakTransform(InTransform=GetDeltaTransformFromRefPose(self=@n2, BoneName=spine_05, BaseName=spine_01)).Rotation)
   > Set backpack_rotate(backpack_rotate=MakeRotator(Roll=MapRangeClamped(Value=@n3.Yaw, InRangeA=0, InRangeB=-60, OutRangeB=50)))
   > Set!(
    execute
    pickaxe_rotate_Roll:float=0
    pickaxe_rotate_Pitch:float=0
    pickaxe_rotate_Yaw:float=MapRangeClamped(Value=@n3, InRangeA=0, InRangeB=20, OutRangeB=-15)
   ) -> (then, Output_Get:Rotator)
   > Set!(
    execute
    belt_rotate_Roll:float=MapRangeClamped(Value=@n3.Yaw, InRangeA=-18, InRangeB=-60, OutRangeB=30)
    belt_rotate_Pitch:float=0
    belt_rotate_Yaw:float=0
   ) -> (then, Output_Get)
   > Set!(
    execute
    belt_pickaxe_Y_Roll:float=0
    belt_pickaxe_Y_Pitch:float=0
    belt_pickaxe_Y_Yaw:float=MapRangeClamped(Value=@n3, InRangeA=-5, InRangeB=-42, OutRangeB=-20)
   ) -> (then, Output_Get)
   > Set!(execute, RBAN_on_Roll:float=0, RBAN_on_Pitch:float=0, RBAN_on_Yaw:float=MapRangeClamped(Value=@n3, InRangeA=0, InRangeB=-26, OutRangeB=-15)) -> (then, Output_Get)
   > Set RBAN_of(RBAN_of:bool=NOT Boolean(A=Get bRigidBodyEnabled())) -> (Output_Get:bool)
   Reroute Node(InputPin=@n1.then_3)
   Set!(
    execute=@n1.then_1
    backpack_translation_X:double=MapRangeClamped(Value=@n3.Yaw, InRangeA=0, InRangeB=0, OutRangeB=0)
    backpack_translation_Y:double=MapRangeClamped(Value=@n3.Yaw, InRangeA=-20, InRangeB=-60, OutRangeB=-15)
    backpack_translation_Z:double=MapRangeClamped(Value=@n3.Yaw, InRangeA=0, InRangeB=-60)
   ) -> (then, Output_Get:Vector)
   n4 = MapRangeClamped(Value=@n3.Yaw, InRangeA=0, InRangeB=-60, OutRangeA=1, OutRangeB=0.9)
   > Set backpack_scale(backpack_scale:Vector=MakeVector(X=@n4, Y=@n4, Z=@n4)) -> (Output_Get:Vector)
   n5 = "Gravity Override State Function (Parameter Asset) (Nativized)"(
    execute=@n1.then_2
    CharacterPartAnimInstance=Self-Reference()
    GameTimeInSeconds=GetGameTimeInSeconds()
    Parameters=&/F_MED_CopperMoggy_Prime_Pack_Parameters
   )
   > Set Gravity Override(Gravity Override=@n5)
   n6 = BreakTransform(InTransform=GetDeltaTransformFromRefPose(self=GetOwningComponent(), BoneName=base_belt, BaseName=dyn_belt_01))
   n7 = MapRangeClamped(Value="Absolute (Float)"(A=BreakRotator(InRot=@n6.Rotation).Yaw), InRangeA=35, InRangeB=120)
   > Set Sim Space Settings(Sim Space Settings:Sim Space Settings Structure=Make Sim Space Settings!(
     WorldAlpha:float=Lerp(A=1, B=0.3, Alpha=@n7)
     VelocityScaleZ:float=Lerp(A=1, B=0.3, Alpha=@n7)
     DampingAlpha:float=Lerp(A=0, B=0.3, Alpha=@n7)
     MaxLinearVelocity:float=350
     MaxAngularVelocity:float=2
     MaxLinearAcceleration:float=1200
     MaxAngularAcceleration:float=8
     ExternalLinearDragV:Vector="0.5,0.5,1"
     ExternalLinearVelocity:Vector="0,0,0"
     ExternalAngularVelocity:Vector="0,0,0"
    ) -> (SimSpaceSettings:Sim Space Settings Structure) {is=title K2Node_MakeStruct}) -> (Output_Get:Sim Space Settings Structure)
   TryGetPawnOwner()
  ]}
]
variables [
 backpac {type=Rotator}
 backpack_rotate {type=Rotator}
 backpack_scale {type=Vector}
 backpack_translation {type=Vector}
 belt_pickaxe {type=Rotator}
 belt_pickaxe_Y {type=Rotator}
 belt_rotate {type=Rotator}
 Gravity Override {type=Vector}
 pickaxe_rotate {type=Rotator}
 RBAN_of {type=boolean}
 RBAN_on {type=Rotator}
 Sim Space Settings {
  type=SimSpaceSettings
  default {
   dampingAlpha=1
   maxAngularAcceleration=10000
   maxAngularVelocity=10000
   maxLinearAcceleration=10000
   maxLinearVelocity=10000
   velocityScaleZ=1
   worldAlpha=0
   externalLinearDragV=<0 0 0>
  }
 }
]
